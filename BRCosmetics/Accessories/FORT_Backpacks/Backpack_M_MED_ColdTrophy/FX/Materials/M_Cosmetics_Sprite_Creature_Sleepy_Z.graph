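# Unlit translucent Niagara sprite material that draws a sleepy "Z" creature sprite for the ColdTrophy backpack effects.
# Blurred noise, offset by dynamic RandomX on both axes and scaled by Distort, warps the sprite through center-based UV scaling.
# Wispy fire noise, driven by Dissolve, is subtracted from the sprite mask to erode it.
# Particle color feeds emissive adjustments; particle alpha scales the mask before depth fade softens intersections.
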Material /BRCosmetics/Accessories/FORT_Backpacks/Backpack_M_MED_ColdTrophy/FX/Materials/M_Cosmetics_Sprite_Creature_Sleepy_Z
graph {
 properties {
  bAutomaticallySetUsageInEditor=false
  bUsedWithNiagaraSprites=true
  blendMode=BLEND_Translucent
  shadingModel=MSM_Unlit
  usedShadingModels=Unlit
 }
 nodes [
  n1 = ParticleColor()
  n2 = Constant() {r=0.5}
  n3 = DynamicParameter() {paramNames [RandomX,RandomY,Distort,Dissolve]}
  n4 = Add(A=Multiply(A=TextureCoordinate()) {constB=0.75}, B=AppendVector(A=@n3.RandomX, B=@n3.RandomX))
  n5 = Multiply(
   A=Power(Base=TextureSample(UVs=@n4) {samplerType=SAMPLERTYPE_Grayscale, texture=&LowResBlurredNoise}.R)
   B=@n3.Distort
  )
  n6 = TextureSample(UVs=MaterialFunctionCall(UVs=TextureCoordinate(), Texture Scale=Add(A=@n2, B=@n5)) {materialFunction=&ScaleUVsByCenter}.UVs) {
   samplerSource=SSM_Clamp_WorldGroupSettings
   texture=&/T_Cosmetics_Sprite_Creature_Sleepy_Z
  }
  n7 = Multiply(A=@n4) {constB=0.5}
  n8 = LinearInterpolate(
   B=SmoothStep(
    Min=Subtract(B=@n3.Dissolve) {constA=0}
    Value=TextureSample(UVs=@n7) {samplerType=SAMPLERTYPE_Grayscale, texture=/Game/Effects/Fort_Effects/Textures/Fire/T_WispyFireNoise_Soft_M}.R
   )
   Alpha=@n3.Dissolve
  )
  DepthFade(Opacity=Multiply(A=@n1.A, B=Subtract(A=@n6.R, B=@n8)), FadeDistance=ScalarParameter() {defaultValue=20, parameterName=DepthFade}) {fadeDistanceDefault=10}
  MaterialFunctionCall(Emissive=@n1.RGB) {materialFunction=&MF_FX_EmissiveAdjustments}
 ]
}
