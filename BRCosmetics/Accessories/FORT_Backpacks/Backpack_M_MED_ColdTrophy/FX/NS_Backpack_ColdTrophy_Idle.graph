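# Idle Niagara effect for the ColdTrophy backpack cosmetic, producing a persistent pale-blue glow attached to the backpack's skeletal mesh.
# A single emitter bursts two particles that never die of age, picking each one's mesh scale and sprite size from arrays by execution index.
# Each update re-pins them to the skeletal mesh with an offset and writes light attributes slightly above the particle.
# Mesh, sprite and light renderers draw the result.
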NiagaraSystem /BRCosmetics/Accessories/FORT_Backpacks/Backpack_M_MED_ColdTrophy/FX/NS_Backpack_ColdTrophy_Idle
dependencies {
 usedDataInterfaces [
  &NiagaraDataInterfaceArrayFloat
  &NiagaraDataInterfaceArrayFloat3
  &NiagaraDataInterfaceSkeletalMesh
 ]
 usedDynamicInputs []
 usedModules [
  &self:SystemSpawnScript.SystemScriptSource.SystemScriptGraph.NiagaraNodeAssignment_0.SetVariables_#1
  &SystemState
 ]
}
emitters {Minimal {
  emitterUpdateScript [
   &EmitterState()
   &SpawnBurst_Instantaneous(Spawn Count=2, Use Spawn Probability=false, UseLoopCountLimit=true)
  ]
  particleSpawnScript [&InitializeParticle(
    Color=<rgba 0.7881 0.9874 1 0.5>
    Lifetime=5
    Mesh Scale=&SelectVectorFromArray(
     Direct Array Index=&ReturnExecIndex()
     Vector Selection Array=DataInterface() {propertyValues {FloatData [<0.5 0.5 0.8>, <0.4 0.4 0.8>]}}
    )
    Mesh Scale Mode=::NewEnumerator4
    Position Offset=<0 -11 -14.5>
    Sprite Size Mode=::NewEnumerator3
    Uniform Sprite Size=&SelectFloatFromArray(
     Direct Array Index=&ReturnExecIndex()
     Float Selection Array=DataInterface() {propertyValues {FloatData [24,18]}}
    )
    UsePositionOffset=false
    Write Material Random=true
   )]
  particleUpdateScript [
   &ParticleState(Kill Particles When Lifetime Has Elapsed=false)
   &FN_Light_Attributes(
    Color=<rgba 0.7952 0.9716 1 1>
    Create Color Attribute=true
    Create Exponent Attribute=true
    Create Position Attribute=true
    Create Radius Attribute=true
    Exponent=3
    Position=&AddVectorToPosition(Position=$Particles.Position, Vector=<0 0 5>)
    Radius=25
    Scale Color=3
   )
   &SkeletalMeshLocation(Bone Sampling Mode=::NewEnumerator4, MaskSpawnGroup=false, Sampled Position Offset=<0 -9 -14.5>, Skeletal Mesh=$System.Skeletal Mesh Reader)
  ]
  rendererClasses [
   &NiagaraMeshRendererProperties
   &NiagaraSpriteRendererProperties
   &NiagaraLightRendererProperties
  ]
  renderers [
   {rendererClass=&NiagaraMeshRendererProperties, rendererIndex=0}
   {rendererClass=&NiagaraSpriteRendererProperties, rendererIndex=1}
   {rendererClass=&NiagaraLightRendererProperties, rendererIndex=2}
  ]
 }}
summary {
 systemName=NS_Backpack_ColdTrophy_Idle
 emitters [{
   emitterName=Minimal
   rendererClasses [
    &NiagaraMeshRendererProperties
    &NiagaraSpriteRendererProperties
    &NiagaraLightRendererProperties
   ]
  }]
 userVariables [User.Variant {type=NiagaraFloat, defaultValue=0, description=""}]
}
