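# Niagara effect for the Earth sprite-creature backpack cosmetic, shedding drifting mesh leaves around a softly lit body.
# Leaves spawn on a ring with owner-speed-gated probability and VariantID-selected colors, then drift under curl noise, drag and gravity, scaled by a region mask.
# Minimal holds two persistent particles on a skeletal mesh sample point, drawing mesh, sprite and light renderers.
# A component-renderer emitter, Minimal001, is disabled.
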NiagaraSystem /BRCosmetics/Accessories/FORT_Backpacks/Backpack_M_MED_ColdTrophy/FX/NS_Backpack_SpriteCreature_Earth
dependencies {
 usedDataInterfaces [
  &NiagaraDataInterfaceCurve
  &NiagaraDataInterfaceArrayFloat
  &NiagaraDataInterfaceArrayFloat3
  &NiagaraDataInterfaceSkeletalMesh
 ]
 usedDynamicInputs [&MakeFloatFromVector]
 usedModules [
  &self:SystemSpawnScript.SystemScriptSource.SystemScriptGraph.NiagaraNodeAssignment_0.SetVariables_#1
  &self:SystemSpawnScript.SystemScriptSource.SystemScriptGraph.NiagaraNodeAssignment_1.SetVariables_#2
  &self:SystemSpawnScript.SystemScriptSource.SystemScriptGraph.NiagaraNodeAssignment_2.SetVariables_#3
  &SystemState
  &self:SystemSpawnScript.SystemScriptSource.SystemScriptGraph.NiagaraNodeAssignment_5.SetVariables_#4
  &self:SystemSpawnScript.SystemScriptSource.SystemScriptGraph.NiagaraNodeAssignment_6.SetVariables_#5
 ]
}
emitters {
 Leaves {
  emitterUpdateScript [
   &EmitterState()
   &SpawnRate(
    Spawn Probability=&Lerp_Float(
     A=0.6
     B=0
     Alpha=&MakeCustomFloatFromBool(
      Bool=&SetBoolByFloatComparison(A=&VectorLength(Vector=$Engine.Owner.Velocity), B=1000)
      True Float=1
      False Float=0
     )
    )
    SpawnRate=3
    Use Spawn Probability=true
   )
  ]
  particleSpawnScript [
   &InitializeParticle(
    Color Maximum=<rgba 0.1676 0.3 0.0353 1>
    Color Minimum=<rgba 0.24 0.6 0.06 1>
    Color Mode=::NewEnumerator2
    Lifetime Max=2.5
    Lifetime Min=1.25
    Lifetime Mode=::NewEnumerator1
    Mesh Scale Mode=::NewEnumerator1
    Mesh Uniform Scale Max=0.15
    Mesh Uniform Scale Min=0.1
    Sprite Rotation Mode=::NewEnumerator2
    UsePositionOffset=false
    Write Material Random=true
   )
   &SkeletalMeshLocation(Bone Sampling Mode=::NewEnumerator4, MaskSpawnGroup=false, Skeletal Mesh=$System.Skeletal Mesh Reader) {enabled=false}
   &ShapeLocation(
    Disc Coverage=1
    MaskSpawnGroup=false
    Non Uniform Scale=<1 0.6 1>
    Offset Coordinate Space=::Local
    Offset Mode=::NewEnumerator0
    Override Seed=false
    Ring Radius=&RandomRangeFloat(Minimum=2, Maximum=3.5)
    Shape Primitive=::NewEnumerator4
    UseOwnerScale=true
   )
   &AddVelocity(
    Module.Use Falloff Distance=false
    Override Seed=false
    Velocity Mode=::NewEnumerator1
    Velocity Speed=&RandomRangeFloat(Minimum=20, Maximum=30)
   )
   &Color(Color=&MergeHelper_ColorSwitch(
     0=&RandomRangeLinearColor(Minimum=<rgba 0.24 0.6 0.06 1>, Maximum=<rgba 0.1676 0.3 0.0353 1>)
     1=&RandomRangeLinearColor(Minimum=<rgba 0.6 0.33 0.06 1>, Maximum=<rgba 0.3 0.2118 0.0353 1>)
     2=<rgba 0 0 0 0>
     3=&RandomRangeLinearColor(Minimum=<rgba 0.7 0.175 0.1925 1>, Maximum=<rgba 0.6 0.1325 0.09 1>)
     4=&RandomRangeLinearColor(Minimum=<rgba 0.223 0.9712 1.2 1>, Maximum=<rgba 0.4862 0.0348 1 1>)
     5=&RandomRangeLinearColor(Minimum=<rgba 0.72 0.7333 0.8 1>, Maximum=<rgba 0.36 0.3867 0.4 1>)
     6=<rgba 0 0 0 0>
     7=&RandomRangeLinearColor(Minimum=<rgba 0.315 0.03 0.6 1>, Maximum=<rgba 0.0683 0.005 0.1 1>)
     Index=&MakeIntFromFloat(Float=$User.VariantID)
     Num=::NewEnumerator6
    ))
  ]
  particleUpdateScript [
   &ParticleState()
   &ScaleMeshSize(Scale Factor=&VectorFromFloat(Value=&FloatFromCurve(
      FloatCurve=DataInterface() {propertyValues {Curve {keys [{interpMode=RCIM_Cubic, tangentMode=RCTM_Break, time=0, value=0}, <at 0.6 1>, {tangentMode=RCTM_User, time=1, value=0}]}}}
      Scale Curve=$System.SystemScale
     )))
   &UpdateMeshOrientation(Rotation Vector=&RandomRangeVector(Minimum=<-1 -1 -1>, Maximum=<1 1 1>))
   &CurlNoiseForce(
    Bias Noise=false
    Noise Strength=&RandomRangeFloat(Minimum=15, Maximum=35)
    Pan Noise=true
    Pan Noise Field=<0 0 1>
   )
   &VortexForce(
    Origin Pull Amount=50
    UseOriginOffset=false
    Vortex Force Amount=&MakeCustomFloatFromBool(Bool=&RandomBool(), True Float=75, False Float=-75)
   ) {enabled=false}
   &Drag(Drag=3, Use Linear Drag=true, Use Rotational Drag=true)
   &GravityForce(Gravity=&RandomRangeVector(Minimum=<15 15 0>, Maximum=<-15 -15 0>))
   &GravityForce(Gravity=&Multiply_VectorByFloat(
     Vector=<0 0 -50>
     Float=&FloatFromCurve(FloatCurve=DataInterface() {propertyValues {Curve {defaultValue=1, keys [{tangentMode=RCTM_User, time=0.3, value=0}, {tangentMode=RCTM_User, time=1, value=1}]}}})
    )) {module=GravityForce001}
   &InheritSourceMovement(
    Applied Position Scale=&VectorFromFloat(Value=&FloatFromCurve(FloatCurve=DataInterface() {propertyValues {Curve {keys [{arriveTangent=-0.1522, interpMode=RCIM_Cubic, leaveTangent=-0.1522, tangentMode=RCTM_User, time=0, value=1}, <at 1 0>]}}}))
    ApplyForce=false
    ApplyPosition=true
    ApplyPositionFalloff=false
   )
   &SolveForcesAndVelocity(Clamp Velocity=false, Limit Acceleration=false)
   &DynamicMaterialParameters(
    Index 0 Param 2=&RandomRangeFloat(Minimum=0.9, Maximum=1.1)
    Index 0 Param 3=&RandomRangeFloat(Minimum=0, Maximum=0.25)
    Index 0 Param 4=1
    Module.Index0Param0Write=true
    Module.Index0Param1Write=true
    Module.Index0Param2Write=true
    Module.Index0Param3Write=true
   )
   &ScaleColor(Scale RGB=<4 5 4>, ScaleA=true, ScaleRGB=true)
   &RegionMask(
    Inner Float=10
    Region Dimensions:Vector3f=<29 29 40>
    Region Falloff=0.2
    Region Type:ENiagaraDebugDrawShape=::NewEnumerator1
    UseOffset=false
    UseRotation=false
   )
   &ScaleMeshSize(Scale Factor=&VectorFromFloat(Value=$Output.RegionMask.RegionMask)) {module=ScaleMeshSize001}
  ]
  rendererClasses [&NiagaraMeshRendererProperties]
  renderers [{rendererClass=&NiagaraMeshRendererProperties, rendererIndex=0}]
 }
 Minimal {
  emitterUpdateScript [
   &EmitterState()
   &SpawnBurst_Instantaneous(Spawn Count=2, Use Spawn Probability=false, UseLoopCountLimit=true)
  ]
  particleSpawnScript [&InitializeParticle(
    Color=<rgba 0.7881 0.9874 1 0.5>
    Lifetime=5
    Mesh Scale=&SelectVectorFromArray(
     Direct Array Index=&ReturnExecIndex()
     Vector Selection Array=DataInterface() {propertyValues {FloatData [<0.5 0.5 0.8>, <0.4 0.4 0.8>]}}
    )
    Mesh Scale Mode=::NewEnumerator4
    Position Offset=<0 -11 -14.5>
    Sprite Size Mode=::NewEnumerator3
    Uniform Sprite Size=&SelectFloatFromArray(
     Direct Array Index=&ReturnExecIndex()
     Float Selection Array=DataInterface() {propertyValues {FloatData [24,18]}}
    )
    UsePositionOffset=false
    Write Material Random=true
   )]
  particleUpdateScript [
   &ParticleState(Kill Particles When Lifetime Has Elapsed=false)
   &FN_Light_Attributes(
    Color=<rgba 0.7952 0.9716 1 1>
    Create Color Attribute=true
    Create Exponent Attribute=true
    Create Position Attribute=true
    Create Radius Attribute=true
    Exponent=3
    Position=&AddVectorToPosition(Position=$Particles.Position, Vector=<0 0 5>)
    Radius=25
    Scale Color=3
   )
   &SkeletalMeshLocation(Bone Sampling Mode=::NewEnumerator4, MaskSpawnGroup=false, Sampled Position Offset=<0 -9 -14.5>, Skeletal Mesh=$System.Skeletal Mesh Reader)
  ]
  rendererClasses [
   &NiagaraMeshRendererProperties
   &NiagaraSpriteRendererProperties
   &NiagaraLightRendererProperties
  ]
  renderers [
   {rendererClass=&NiagaraMeshRendererProperties, rendererIndex=0}
   {rendererClass=&NiagaraSpriteRendererProperties, rendererIndex=1}
   {rendererClass=&NiagaraLightRendererProperties, rendererIndex=2}
  ]
 }
 Minimal001 {
  bEnabled=false
  emitterUpdateScript [
   &EmitterState()
   &SpawnBurst_Instantaneous(Use Spawn Probability=false, UseLoopCountLimit=true)
  ]
  particleSpawnScript [&InitializeParticle(Color Mode=::NewEnumerator0, Lifetime=5, Position Offset=<0 11.2715 0>, UsePositionOffset=true, Write Material Random=true)]
  particleUpdateScript [&ParticleState(Kill Particles When Lifetime Has Elapsed=false)]
  rendererClasses [&NiagaraComponentRendererProperties]
  renderers [{rendererClass=&NiagaraComponentRendererProperties, rendererIndex=0}]
 }
}
summary {
 systemName=NS_Backpack_SpriteCreature_Earth
 emitters [
  {emitterName=Leaves, rendererClasses [&NiagaraMeshRendererProperties]}
  {
   emitterName=Minimal
   rendererClasses [
    &NiagaraMeshRendererProperties
    &NiagaraSpriteRendererProperties
    &NiagaraLightRendererProperties
   ]
  }
  {bEnabled=false, emitterName=Minimal001, rendererClasses [&NiagaraComponentRendererProperties]}
 ]
 userVariables [
  User.Variant {type=NiagaraInt32}
  User.VariantID {type=NiagaraFloat, defaultValue=0, description=""}
 ]
}
