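# Niagara effect for the sleepy sprite-creature backpack cosmetic, combining a glowing creature body with intermittent drifting Z particles.
# Minimal bursts two persistent, differently sized particles that follow a skeletal mesh bone sample and render as mesh, sprite and light.
# Z probabilistically spawns short-lived sprites tinted from a color array indexed by User.VariantID, which grow, slow under drag and fade, with opacity scaled by time of day.
# The User.Variant HSV step is disabled.
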NiagaraSystem /BRCosmetics/Accessories/FORT_Backpacks/Backpack_M_MED_ColdTrophy/FX/NS_Backpack_SpriteCreature_Sleepy
dependencies {
 usedDataInterfaces [
  &NiagaraDataInterfaceArrayColor
  &NiagaraDataInterfaceArrayFloat3
  &NiagaraDataInterfaceCurve
  &NiagaraDataInterfaceColorCurve
  &NiagaraDataInterfaceArrayFloat
  &NiagaraDataInterfaceSkeletalMesh
 ]
 usedDynamicInputs [&MakeFloatFromVector]
 usedModules [
  &self:SystemSpawnScript.SystemScriptSource.SystemScriptGraph.NiagaraNodeAssignment_0.SetVariables_#1
  &SystemState
  &self:SystemSpawnScript.SystemScriptSource.SystemScriptGraph.NiagaraNodeAssignment_6.SetVariables_#2
  &self:SystemSpawnScript.SystemScriptSource.SystemScriptGraph.NiagaraNodeAssignment_5.SetVariables_#3
 ]
}
emitters {
 Minimal {
  emitterUpdateScript [
   &EmitterState()
   &SpawnBurst_Instantaneous(Spawn Count=2, Use Spawn Probability=false, UseLoopCountLimit=true)
  ]
  particleSpawnScript [&InitializeParticle(
    Color=<rgba 0.7881 0.9874 1 0.5>
    Lifetime=5
    Mesh Scale=&SelectVectorFromArray(
     Direct Array Index=&ReturnExecIndex()
     Vector Selection Array=DataInterface() {propertyValues {FloatData [<0.5 0.5 0.8>, <0.4 0.4 0.8>]}}
    )
    Mesh Scale Mode=::NewEnumerator4
    Position Offset=<0 -11 -14.5>
    Sprite Size Mode=::NewEnumerator3
    Uniform Sprite Size=&SelectFloatFromArray(
     Direct Array Index=&ReturnExecIndex()
     Float Selection Array=DataInterface() {propertyValues {FloatData [24,18]}}
    )
    UsePositionOffset=false
    Write Material Random=true
   )]
  particleUpdateScript [
   &ParticleState(Kill Particles When Lifetime Has Elapsed=false)
   &FN_Light_Attributes(
    Color=<rgba 0.7952 0.9716 1 1>
    Create Color Attribute=true
    Create Exponent Attribute=true
    Create Position Attribute=true
    Create Radius Attribute=true
    Exponent=3
    Position=&AddVectorToPosition(Position=$Particles.Position, Vector=<0 0 5>)
    Radius=25
    Scale Color=3
   )
   &SkeletalMeshLocation(Bone Sampling Mode=::NewEnumerator4, MaskSpawnGroup=false, Sampled Position Offset=<0 -9 -14.5>, Skeletal Mesh=$System.Skeletal Mesh Reader)
  ]
  rendererClasses [
   &NiagaraMeshRendererProperties
   &NiagaraSpriteRendererProperties
   &NiagaraLightRendererProperties
  ]
  renderers [
   {rendererClass=&NiagaraMeshRendererProperties, rendererIndex=0}
   {rendererClass=&NiagaraSpriteRendererProperties, rendererIndex=1}
   {rendererClass=&NiagaraLightRendererProperties, rendererIndex=2}
  ]
 }
 Z {
  emitterSpawnScript [/SetVariables(Emitter.EmissiveMultMaster=1) {isSetParameters=true}]
  emitterUpdateScript [
   &EmitterState(Life Cycle Mode=::NewEnumerator1, Loop Duration=1, UseLoopDelay=false)
   &SpawnRate(Spawn Probability=0.75, SpawnRate=0.6, Use Spawn Probability=true)
  ]
  particleSpawnScript [
   &InitializeParticle(
    Color=&SelectLinearColorFromArray(
     Array Sampling Mode=::NewEnumerator1
     Direct Array Index=&MakeIntFromFloat(Float=$User.VariantID)
     Color Selection Array=DataInterface() {propertyValues {ColorData [
        <rgba 0.8 0.8 0.8 1>
        <rgba 1 0.5348 0.0696 1>
        <rgba 1 1 1 1>
        <rgba 1 0.15 0.1944 1>
        <rgba 0.0662 0.045 0.3 1>
        <rgba 1 1 1 1>
        <rgba 1 1 1 1>
        <rgba 0.0347 0.002 0.1 1>
       ]}}
     Random Seed=?
    )
    Lifetime Min=1.5
    Lifetime Mode=::NewEnumerator1
    Mesh Scale=&Multiply_VectorByFloat(
     Vector=&RandomRangeVector(Minimum=<1 1 1>, Maximum=<1 1 1>)
     Float=&RandomRangeFloat(Minimum=0.4, Maximum=0.8)
    )
    Mesh Scale Mode=::NewEnumerator4
    Sprite Size Max=<12 10>
    Sprite Size Min=<10 8>
    Sprite Size Mode=::NewEnumerator2
    UsePositionOffset=false
    Write Material Random=true
   )
   &ShapeLocation(
    MaskSpawnGroup=false
    Non Uniform Scale=<1.3 0.8 0.1>
    Offset Mode=::NewEnumerator0
    Override Seed=false
    Sphere Radius=14
    Sphere Surface Distribution=0.3571
    UseOwnerScale=true
   )
   &AddVelocity(
    Module.Use Falloff Distance=false
    Origin Offset=<0 0 -3>
    Override Seed=false
    Velocity Mode=::NewEnumerator1
    Velocity Speed=&Multiply_Float(A=&RandomRangeFloat(Minimum=35, Maximum=40), B=$System.SystemScale)
   )
   /SetVariables(Particles.HSV:Vector3f=&SelectVectorFromArray(
     Direct Array Index=$User.Variant
     Vector Selection Array=DataInterface() {propertyValues {FloatData [<1 1 1>, <0.6 0.5 1.25>, <0.7 0.7 1>, <0.05 0.2 1>, <0.02 1 0.8>, <1 1 1>]}}
    )) {enabled=false, isSetParameters=true}
  ]
  particleUpdateScript [
   &ParticleState()
   &ScaleSpriteSize(
    Uniform Curve Scale=$System.SystemScale
    Uniform Curve Sprite Scale=DataInterface() {propertyValues {Curve {
       defaultValue=0
       keys [
        {arriveTangent=4.1785, interpMode=RCIM_Cubic, leaveTangent=4.1785, tangentMode=RCTM_User, time=0, value=0}
        {arriveTangent=1.1179, interpMode=RCIM_Cubic, leaveTangent=1.1179, tangentMode=RCTM_User, time=0.2042, value=0.6228}
        {arriveTangent=0.0224, interpMode=RCIM_Cubic, leaveTangent=0.0224, tangentMode=RCTM_User, time=1, value=1}
       ]
      }}}
   )
   &InheritSourceMovement(
    Applied Position Scale=&Multiply_VectorByFloat(Vector=<1 1 1>, Float=&Lerp_Float(A=1, B=0.98, Alpha=$Particles.NormalizedAge))
    ApplyForce=false
    ApplyPosition=true
    ApplyPositionFalloff=false
   ) {enabled=false, module=InheritSourceMovement001}
   &Drag(Drag=&RandomRangeFloat(Minimum=3, Maximum=3.25), Use Linear Drag=true, Use Rotational Drag=true)
   &SolveForcesAndVelocity(Clamp Velocity=false, Limit Acceleration=false)
   &DynamicMaterialParameters(
    Index 0 Param 1=&RandomRangeFloat(Minimum=0, Maximum=1)
    Index 0 Param 2=&RandomRangeFloat(Minimum=0, Maximum=1)
    Index 0 Param 3=&FloatFromCurve(
     FloatCurve=DataInterface() {propertyValues {Curve {
        defaultValue=0
        keys [
         {arriveTangent=0.299, interpMode=RCIM_Cubic, leaveTangent=0.299, tangentMode=RCTM_User, time=0, value=0}
         {arriveTangent=1.6382, tangentMode=RCTM_User, time=1, value=1}
        ]
       }}}
     Scale Curve=3.5
    )
    Index 0 Param 4=&FloatFromCurve(
     FloatCurve=DataInterface() {propertyValues {Curve {
        defaultValue=0
        keys [{interpMode=RCIM_Cubic, tangentMode=RCTM_User, time=0, value=0}, {arriveTangent=-0.0351, tangentMode=RCTM_User, time=1, value=1}]
       }}}
     Scale Curve=2
    )
    Module.Index0Param0Write=true
    Module.Index0Param1Write=true
    Module.Index0Param2Write=true
    Module.Index0Param3Write=true
   )
   &ScaleColor(
    Scale Alpha=&FloatFromCurve(
     FloatCurve=DataInterface() {propertyValues {Curve {
        defaultValue=0
        keys [
         {arriveTangent=5.9724, interpMode=RCIM_Cubic, leaveTangent=5.9724, tangentMode=RCTM_User, time=0, value=0}
         {interpMode=RCIM_Cubic, tangentMode=RCTM_User, time=0.5, value=1}
         {arriveTangent=-6.0921, tangentMode=RCTM_User, time=1, value=0}
        ]
       }}}
     Scale Curve=&FNTimeOfDayScalar(Night Scale=0.4, Morning Scale=0.7, Day Scale=0.9, Evening Scale=0.7)
    )
    Scale RGB=&MakeVectorFromLinearColorRGB(Color=&ColorFromCurve(ColorCurve=DataInterface() {propertyValues {
        AlphaCurve {keys [<at 0 1>,<at 1 0>]}
        BlueCurve {keys [<at 0.1355 1.5>, {arriveTangent=3.3726, time=0.2116, value=1.5}, {arriveTangent=-0.2414, time=0.4208, value=1}, <at 0.9059 0.02>]}
        GreenCurve {keys [<at 0.1355 2.5>, {arriveTangent=-7.3404, time=0.2116, value=1.5}, {arriveTangent=-0.9957, time=0.4208, value=1}, <at 0.9059 0.02>]}
        RedCurve {keys [<at 0.1355 3>, {arriveTangent=-11.1098, time=0.2116, value=1.5}, {arriveTangent=-1.9572, time=0.4208, value=1}, <at 0.9059 0.02>]}
       }}))
    ScaleA=true
    ScaleRGB=true
   )
   &InterpolateOverTime(Rate Of Change=12, Target Position Value:NiagaraPosition=$Particles.Position, Variable Type=::NewEnumerator6) {enabled=false}
  ]
  rendererClasses [&NiagaraSpriteRendererProperties]
  renderers [{rendererClass=&NiagaraSpriteRendererProperties}]
 }
}
summary {
 systemName=NS_Backpack_SpriteCreature_Sleepy
 emitters [
  {emitterName=Z, rendererClasses [&NiagaraSpriteRendererProperties]}
  {
   emitterName=Minimal
   rendererClasses [
    &NiagaraMeshRendererProperties
    &NiagaraSpriteRendererProperties
    &NiagaraLightRendererProperties
   ]
  }
 ]
 userVariables [
  User.Variant {type=NiagaraInt32, defaultValue=1}
  User.VariantID {type=NiagaraFloat, defaultValue=0, description=""}
 ]
}
