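# Niagara effect for the water variant of a sprite-creature backpack cosmetic, layering mist, droplets, dangling water meshes and a glow.
# Minimal bursts two persistent mesh, sprite and light particles that follow the skeletal mesh.
# Water_Pendulum001 uses spring-driven pendulums, passing their direction to materials, while WaterDroplets_Small emits short-lived drifting meshes.
# User.VariantID sets droplet and pendulum renderer visibility tags and hue-shifts Water_Mist, whose opacity follows time of day.
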NiagaraSystem /BRCosmetics/Accessories/FORT_Backpacks/Backpack_M_MED_ColdTrophy/FX/NS_Backpack_SpriteCreature_Water
dependencies {
 usedDataInterfaces [
  &NiagaraDataInterfaceArrayFloat3
  &NiagaraDataInterfaceColorCurve
  &NiagaraDataInterfaceCurve
  &NiagaraDataInterfaceArrayFloat
  &NiagaraDataInterfaceSkeletalMesh
 ]
 usedDynamicInputs [&SubtractVectorFromPosition, &TransformPosition]
 usedModules [
  &self:SystemSpawnScript.SystemScriptSource.SystemScriptGraph.NiagaraNodeAssignment_1.SetVariables_#1
  &self:SystemSpawnScript.SystemScriptSource.SystemScriptGraph.NiagaraNodeAssignment_0.SetVariables_#2
  &SystemState
  &self:SystemSpawnScript.SystemScriptSource.SystemScriptGraph.NiagaraNodeAssignment_5.SetVariables_#3
  &self:SystemSpawnScript.SystemScriptSource.SystemScriptGraph.NiagaraNodeAssignment_6.SetVariables_#4
 ]
}
emitters {
 Minimal {
  emitterUpdateScript [
   &EmitterState()
   &SpawnBurst_Instantaneous(Spawn Count=2, Use Spawn Probability=false, UseLoopCountLimit=true)
  ]
  particleSpawnScript [&InitializeParticle(
    Color=<rgba 0.7881 0.9874 1 0.5>
    Lifetime=5
    Mesh Scale=&SelectVectorFromArray(
     Direct Array Index=&ReturnExecIndex()
     Vector Selection Array=DataInterface() {propertyValues {FloatData [<0.5 0.5 0.8>, <0.4 0.4 0.8>]}}
    )
    Mesh Scale Mode=::NewEnumerator4
    Position Offset=<0 -11 -14.5>
    Sprite Size Mode=::NewEnumerator3
    Uniform Sprite Size=&SelectFloatFromArray(
     Direct Array Index=&ReturnExecIndex()
     Float Selection Array=DataInterface() {propertyValues {FloatData [10,5]}}
    )
    UsePositionOffset=false
    Write Material Random=true
   )]
  particleUpdateScript [
   &ParticleState(Kill Particles When Lifetime Has Elapsed=false)
   &FN_Light_Attributes(
    Color=<rgba 0.7952 0.9716 1 1>
    Create Color Attribute=true
    Create Exponent Attribute=true
    Create Position Attribute=true
    Create Radius Attribute=true
    Exponent=3
    Position=&AddVectorToPosition(Position=$Particles.Position, Vector=<0 0 5>)
    Radius=25
    Scale Color=3
   )
   &SkeletalMeshLocation(Bone Sampling Mode=::NewEnumerator4, MaskSpawnGroup=false, Sampled Position Offset=<0 -9 -14.5>, Skeletal Mesh=$System.Skeletal Mesh Reader)
  ]
  rendererClasses [
   &NiagaraMeshRendererProperties
   &NiagaraSpriteRendererProperties
   &NiagaraLightRendererProperties
  ]
  renderers [
   {rendererClass=&NiagaraMeshRendererProperties, rendererIndex=0}
   {rendererClass=&NiagaraSpriteRendererProperties, rendererIndex=1}
   {rendererClass=&NiagaraLightRendererProperties, rendererIndex=2}
  ]
 }
 WaterDroplets_Small {
  emitterSpawnScript [/SetVariables(Emitter.EmissiveMultMaster:NiagaraFloat=$User.EmissiveMultMaster) {isSetParameters=true}]
  emitterUpdateScript [
   &EmitterState(Life Cycle Mode=::NewEnumerator1, Loop Duration=1, UseLoopDelay=false)
   &SpawnRate(Use Spawn Probability=false)
  ]
  particleSpawnScript [
   &InitializeParticle(
    Color Mode=::NewEnumerator0
    Lifetime Max=0.7
    Lifetime Min=1.3
    Lifetime Mode=::NewEnumerator1
    Mesh Scale=&VectorFromFloat(Value=&RandomRangeFloat(Minimum=0.4, Maximum=0.5))
    Mesh Scale Mode=::NewEnumerator4
    UsePositionOffset=true
    Write Material Random=true
   )
   &ShapeLocation(
    MaskSpawnGroup=false
    Non Uniform Scale=<1.3 0.8 0.1>
    Offset Mode=::NewEnumerator0
    Override Seed=false
    Sphere Radius=8
    UseOwnerScale=true
   )
   &AddVelocity(
    Module.Use Falloff Distance=false
    Origin Offset=<0 0 50>
    Override Seed=false
    Velocity Mode=::NewEnumerator1
    Velocity Speed=&Multiply_Float(A=-15, B=$System.SystemScale)
   )
   &InitialMeshOrientation(Rotation=&RandomRangeVector(Minimum=<0 0 0>, Maximum=<0 0 1>), Use Rotation Vector=true)
  ]
  particleUpdateScript [
   /SetVariables(Particles.VisibilityTag=&MakeIntFromFloat(Float=$User.VariantID)) {isSetParameters=true}
   &ParticleState()
   &Drag(Drag=2, Use Linear Drag=true, Use Rotational Drag=true)
   &SolveForcesAndVelocity(Clamp Velocity=false, Limit Acceleration=false)
   &DynamicMaterialParameters(
    Index 2 Param 1=&RandomRangeFloat(Minimum=1, Maximum=1)
    Index 2 Param 2=&RandomRangeFloat(Minimum=0.5, Maximum=0.3)
    Index 3 Param 1=1
    Index 3 Param 2=1
    Index 3 Param 3=1
    Index 3 Param 4=&FloatFromCurve(FloatCurve=DataInterface() {propertyValues {Curve {keys [<at 0 0>,<at 0.2 1>,<at 0.8 1>,<at 1 0>]}}}, Scale Curve=0.4)
    Module.Index2Param0Write=true
    Module.Index2Param1Write=true
    Module.Index2Param2Write=true
    Module.Index2Param3Write=true
    Module.Index3Param0Write=true
    Module.Index3Param1Write=true
    Module.Index3Param2Write=true
    Module.Index3Param3Write=true
    Write Parameter Index 0=false
    Write Parameter Index 2=true
    Write Parameter Index 3=true
   )
   &ScaleMeshSize(Scale Factor=&VectorFromFloat(Value=&FloatFromCurve(
      FloatCurve=DataInterface() {propertyValues {Curve {keys [
          {interpMode=RCIM_Cubic, tangentMode=RCTM_User, time=0, value=0}
          {arriveTangent=0.1433, interpMode=RCIM_Cubic, leaveTangent=0.1433, tangentMode=RCTM_User, time=0.1, value=1}
          {arriveTangent=-1.7257, interpMode=RCIM_Cubic, leaveTangent=-1.7257, tangentMode=RCTM_User, time=0.5707, value=0.9603}
          {arriveTangent=-2.3817, interpMode=RCIM_Cubic, tangentMode=RCTM_Break, time=1, value=0}
         ]}}}
      Scale Curve=0.4
     ))) {module=ScaleMeshSize001}
   &ScaleColor(
    Scale Alpha=&FloatFromCurve(
     FloatCurve=DataInterface() {propertyValues {Curve {keys [{interpMode=RCIM_Cubic, tangentMode=RCTM_User, time=0, value=0.3}, {interpMode=RCIM_Cubic, tangentMode=RCTM_User, time=0.5, value=1}]}}}
     Scale Curve=0.9
    )
    ScaleA=true
    ScaleRGB=true
   )
  ]
  rendererClasses [&NiagaraMeshRendererProperties]
  renderers [
   {rendererClass=&NiagaraMeshRendererProperties, rendererIndex=0}
   {rendererClass=&NiagaraMeshRendererProperties, rendererIndex=1}
   {rendererClass=&NiagaraMeshRendererProperties, rendererIndex=2}
   {rendererClass=&NiagaraMeshRendererProperties, rendererIndex=3}
   {rendererClass=&NiagaraMeshRendererProperties, rendererIndex=4}
   {rendererClass=&NiagaraMeshRendererProperties, rendererIndex=5}
  ]
 }
 Water_Mist {
  emitterUpdateScript [
   &EmitterState()
   &SpawnRate(SpawnRate=15, Use Spawn Probability=false) {module=SpawnRate001}
  ]
  particleSpawnScript [
   &InitializeParticle(
    Sprite Rotation Mode=::NewEnumerator2
    Sprite Size Mode=::NewEnumerator3
    Sprite UV Mode=::NewEnumerator3
    Uniform Sprite Size=&Multiply_Float(A=20, B=$System.SystemScale)
    UsePositionOffset=true
    Write Material Random=true
   )
   &ShapeLocation(
    MaskSpawnGroup=false
    Non Uniform Scale=<2 1 0.2>
    Offset=<0 0 -2>
    Offset Mode=::NewEnumerator0
    Override Seed=false
    Sphere Radius=3
    UseOwnerScale=true
   )
   &AddVelocity(Velocity=<0 0 100>, Velocity Speed Scale=&Multiply_Float(A=1, B=$System.SystemScale))
  ]
  particleUpdateScript [
   /SetVariables(Particles.HSV:Vector3f=&SelectVectorFromArray(
     Direct Array Index=&MakeIntFromFloat(Float=$User.VariantID)
     Vector Selection Array=DataInterface() {propertyValues {FloatData [<1 1 1>, <0.45 0.7 1>, <0.95 1 0.8>, <0.55 1 0.8>, <0.86 1.5 0.5>, <1 0.2 1>, <-0.3 1 1>, <0 0 0>]}}
    )) {isSetParameters=true}
   &ParticleState()
   &SkeletalMeshLocation(
    Bone Sampling Mode=::NewEnumerator4
    "Inherit Velocity (Scale)"=0
    MaskSpawnGroup=false
    Position Sampling=::NewEnumerator1
    Sampled Position Offset=<0 0 -20>
    Skeletal Mesh=$System.Skeletal Mesh Reader
    Velocity Sampling=::NewEnumerator0
    VelocitySpeedLimit=false
   ) {enabled=false}
   &ScaleColor(
    Scale Alpha=&FloatFromCurve(
     FloatCurve=DataInterface() {propertyValues {Curve {keys [<at 0 0 RCIM_Cubic>, {interpMode=RCIM_Cubic, tangentMode=RCTM_Break, time=0.2, value=1}, <at 1 0>]}}}
     Scale Curve=&FNTimeOfDayScalar(Night Scale=0.15, Morning Scale=0.25, Day Scale=0.5, Evening Scale=0.25)
    )
    Scale RGB=&MakeVectorFromLinearColorRGB(Color=&HueShiftLinearColor(
      Hue=&MakeFloatFromVector(Vector=$Particles.HSV, Channel=::NewEnumerator0)
      Saturation=&MakeFloatFromVector(Vector=$Particles.HSV, Channel=::NewEnumerator1)
      Value=&MakeFloatFromVector(Vector=$Particles.HSV, Channel=::NewEnumerator2)
      Linear Color=&ColorFromCurve(ColorCurve=DataInterface() {propertyValues {
         AlphaCurve {keys []}
         BlueCurve {keys [<at 0.0767 3>,<at 0.1758 2>,<at 0.6485 1>]}
         GreenCurve {keys [<at 0.0767 2.36>,<at 0.1758 1.4984>,<at 0.6485 0.4107>]}
         RedCurve {keys [<at 0.0767 0.6>,<at 0.1758 0.1191>,<at 0.6485 0>]}
        }})
     ))
    ScaleA=true
    ScaleRGB=true
   ) {module=ScaleColor001}
   &InheritSourceMovement(
    Applied Position Scale=&Multiply_VectorByFloat(Vector=<1 1 1>, Float=&Lerp_Float(A=1, B=0.9, Alpha=$Particles.NormalizedAge))
    ApplyForce=false
    ApplyPosition=true
    ApplyPositionFalloff=false
    Source Velocity:Vector3f=$Particles.SkeletalMeshLocation.SampledVelocity
   ) {enabled=false, module=InheritSourceMovement001}
   &Drag(Drag=22, Use Linear Drag=true, Use Rotational Drag=true) {enabled=false}
   &ScaleSpriteSize(Uniform Curve Sprite Scale=DataInterface() {propertyValues {Curve {keys [{tangentMode=RCTM_User, time=0, value=1}, {tangentMode=RCTM_User, time=1, value=0.5}]}}})
   &GravityForce(Gravity=<0 0 5000>)
   &SolveForcesAndVelocity(Clamp Velocity=true, Limit Acceleration=false, Speed Limit=9)
   &DynamicMaterialParameters(Module.Index0Param0Write=true, Module.Index0Param1Write=true, Module.Index0Param2Write=false, Module.Index0Param3Write=false)
   &CameraOffset(Camera Offset Amount=&RandomRangeFloat(Minimum=9, Maximum=11))
  ]
  rendererClasses [&NiagaraSpriteRendererProperties]
  renderers [{rendererClass=&NiagaraSpriteRendererProperties}]
 }
 Water_Pendulum001 {
  emitterSpawnScript [/SetVariables(Emitter.EmissiveMultMaster:NiagaraFloat=$User.EmissiveMultMaster) {isSetParameters=true}]
  emitterUpdateScript [
   &EmitterState(Life Cycle Mode=::NewEnumerator1, Loop Duration=1, UseLoopDelay=false)
   &SpawnRate(SpawnRate=&RandomRangeFloat(Minimum=1, Maximum=1.2), Use Spawn Probability=false)
   &SpawnRate(
    SpawnRate=&FloatFromCurve(
     FloatCurve=DataInterface() {propertyValues {Curve {keys [
         {interpMode=RCIM_Cubic, tangentMode=RCTM_User, time=0, value=1}
         {arriveTangent=-8.7519, interpMode=RCIM_Cubic, tangentMode=RCTM_Break, time=0.35, value=0}
        ]}}}
     CurveIndex=$Emitter.Age
     Scale Curve=8
    )
    Use Spawn Probability=false
   ) {module=SpawnRate001}
  ]
  particleSpawnScript [
   &InitializeParticle(
    Lifetime=&RandomRangeFloat(Minimum=2, Maximum=4)
    Mesh Scale=&Multiply_VectorByFloat(Vector=<1 1 1>, Float=0.15)
    Mesh Scale Mode=::NewEnumerator4
    Position Offset=<0 0 30>
    UsePositionOffset=false
    Write Material Random=true
   )
   /SetVariables(Particles.PendulumLength=5, Particles.PendulumVector:Vector3f=<0 0 1>) {isSetParameters=true}
  ]
  particleUpdateScript [
   /SetVariables(Particles.VisibilityTag=&MakeIntFromFloat(Float=$User.VariantID)) {isSetParameters=true}
   &SkeletalMeshLocation(
    Bone Sampling Mode=::NewEnumerator4
    MaskSpawnGroup=false
    Position Sampling=::NewEnumerator1
    Sampled Position Offset=<0 1 0>
    Skeletal Mesh=$System.Skeletal Mesh Reader
   ) {enabled=false}
   &ParticleState()
   &UpdateMeshOrientation(Rotation Rate=5) {enabled=false}
   /SetVariables(Particles.PendulumLength:NiagaraFloat=&FloatFromCurve(
     FloatCurve=DataInterface() {propertyValues {Curve {keys [{interpMode=RCIM_Cubic, leaveTangent=2.3132, tangentMode=RCTM_Break, time=0, value=0}, {tangentMode=RCTM_User, time=1, value=1}]}}}
     Scale Curve=3
    )) {isSetParameters=true}
   &PendulumSetup(
    Module.Enable Spring Driver=true
    Pendulum Length:NiagaraFloat=$Particles.PendulumLength
    Pendulum Pivot:Vector3f=&AddVectorToPosition(Position=$Engine.Emitter.SimulationPosition, Vector=<0 0 0>)
    Pendulum Pivot Offset:Vector3f=&RandomRangeVector(Minimum=<-6 -2 1>, Maximum=<6 4 1>)
    Pendulum Pivot Offset Coordinate Space:ENiagaraCoordinateSpace=::Local
    Pendulum Rest Axis:Vector3f=<0 0 1>
    Rigid Pendulum=false
    "Spring Driven Constraint (Tightness)"=5
   )
   &Drag(Drag=22, Use Linear Drag=true, Use Rotational Drag=true) {module=Drag001}
   &SolveForcesAndVelocity(Clamp Velocity=false, Limit Acceleration=false)
   &PendulumConstraint()
   &CalculateAccurateVelocity()
   /SetVariables(Particles.DynamicMaterialParameter1:Vector4f=&MakeFloat4FromVectorAndFloat(
     "Vector (RGB)"=&TransformVector(Vector=/Niagara/DynamicInputs/Vectors/NegateVector(Vector:Vector3f=$Particles.PendulumVector), Destination Space=::World)
     "Float (Alpha)"=&FloatFromCurve(
      FloatCurve=DataInterface() {propertyValues {Curve {keys [
          {arriveTangent=5.9171, interpMode=RCIM_Cubic, time=0, value=0}
          <at 0.2 1 RCIM_Cubic>
          {arriveTangent=-6.1005, interpMode=RCIM_Cubic, leaveTangent=-6.1005, tangentMode=RCTM_User, time=0.3, value=0}
         ]}}}
      Scale Curve=4.5
     )
    )) {isSetParameters=true}
   &DynamicMaterialParameters(
    Index 0 Param 3=0.5
    Index 2 Param 1=&Lerp_Float(
     A=0
     B=2
     Alpha=&FloatFromCurve(FloatCurve=DataInterface() {propertyValues {Curve {keys [
          {arriveTangent=4.5376, interpMode=RCIM_Cubic, leaveTangent=4.5376, tangentMode=RCTM_User, time=0.2, value=0}
          {interpMode=RCIM_Cubic, leaveTangent=-5.7375, tangentMode=RCTM_Break, time=0.5, value=0.5}
          {interpMode=RCIM_Cubic, tangentMode=RCTM_User, time=0.7, value=0}
         ]}}})
    )
    Index 2 Param 2=0.5
    Index 2 Param 4=&FloatFromCurve(
     FloatCurve=DataInterface() {propertyValues {Curve {keys [{interpMode=RCIM_Cubic, tangentMode=RCTM_User, time=0, value=0}, {interpMode=RCIM_Cubic, tangentMode=RCTM_User, time=1, value=1}]}}}
     Scale Curve=0.2
    )
    Index 3 Param 1=&RandomRangeFloat(Minimum=1, Maximum=1.5)
    Index 3 Param 2=1
    Index 3 Param 3=&RandomRangeFloat(Minimum=1.5, Maximum=1)
    Index 3 Param 4=0.5
    Module.Index0Param0Write=false
    Module.Index0Param1Write=false
    Module.Index0Param2Write=true
    Module.Index0Param3Write=true
    Module.Index2Param0Write=true
    Module.Index2Param1Write=true
    Module.Index2Param2Write=false
    Module.Index2Param3Write=true
    Module.Index3Param0Write=true
    Module.Index3Param1Write=true
    Module.Index3Param2Write=true
    Module.Index3Param3Write=true
    Write Parameter Index 2=true
    Write Parameter Index 3=true
   ) {module=DynamicMaterialParameters001}
   &ScaleMeshSize(Scale Factor=&VectorFromFloat(Value=&FloatFromCurve(FloatCurve=DataInterface() {propertyValues {Curve {keys [{interpMode=RCIM_Cubic, leaveTangent=-6.3809, tangentMode=RCTM_Break, time=0.453, value=1}, <at 1.0031 0.0119>]}}}))) {module=ScaleMeshSize001}
   &CameraOffset()
  ]
  rendererClasses [&NiagaraMeshRendererProperties]
  renderers [
   {rendererClass=&NiagaraMeshRendererProperties, rendererIndex=0}
   {rendererClass=&NiagaraMeshRendererProperties, rendererIndex=1}
   {rendererClass=&NiagaraMeshRendererProperties, rendererIndex=2}
   {rendererClass=&NiagaraMeshRendererProperties, rendererIndex=3}
   {rendererClass=&NiagaraMeshRendererProperties, rendererIndex=4}
   {rendererClass=&NiagaraMeshRendererProperties, rendererIndex=5}
  ]
 }
}
summary {
 systemName=NS_Backpack_SpriteCreature_Water
 emitters [
  {emitterName=Water_Mist, rendererClasses [&NiagaraSpriteRendererProperties]}
  {emitterName=WaterDroplets_Small, rendererClasses [&NiagaraMeshRendererProperties]}
  {emitterName=Water_Pendulum001, rendererClasses [&NiagaraMeshRendererProperties]}
  {
   emitterName=Minimal
   rendererClasses [
    &NiagaraMeshRendererProperties
    &NiagaraSpriteRendererProperties
    &NiagaraLightRendererProperties
   ]
  }
 ]
 userVariables [
  User.EmissiveMultMaster {type=NiagaraFloat, defaultValue=1, description=""}
  User.VariantID {type=NiagaraFloat, defaultValue=0, description=""}
 ]
}
