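# Physics asset for the ColdTrophy backpack cosmetic mesh, giving the back bling a simple collision and physics shape.
# It holds a single body on the pack bone, built from one capsule offset and rotated to cover the pack.
# It defines no constraints, so the pack has no jointed or simulated secondary parts.
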PhysicsAsset /BRCosmetics/Accessories/FORT_Backpacks/Backpack_M_MED_ColdTrophy/Meshes/M_MED_ColdTrophy_Pack_PhysicsAsset
bodies {pack [{shapeType=Capsule, length=17.9232, radius=15.8079, center=<0.0519 -6.7661 -0.0794>, rotation=<rot 0 -90.7314 -179.7117>}]}
constraints []
